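; Cooler loot container that adds an open sound for third-person observers when someone loots it.
; On loot, it runs the parent logic, then calls the sound event if the searching pawn is valid and not locally viewed.
; That event plays the item-interact sound through the SoundLibrary subsystem where cosmetic events can execute.
; The construction script adds a Lackluster tiered-ammo material tag to the actor.
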(event EventOnLoot
  ;[0.15 0.1047 0.003615 0.5] Cast Open Sound to 3P
  (|CallParentFunction)
  (bind _returnvalue (Container|GetSearchingPawn))
  (Utilities|IsValid _returnvalue
    (:Is%20Valid
      (if (not (Pawn|IsLocallyViewed _returnvalue))
        (CallFunction|PlayOpenSound3p)))))

(event Custom|Play%20Open%20Sound%203p
  ;[0.15 0.1047 0.003615 0.5] Cast Open Sound to 3P
  (Networking|CanExecuteCosmeticEvents
    (:True
      (SoundLibrary|PlaySoundSimple (WorldSubsystems|GetSoundLibrarySubsystem) self (Variables|Default|GetItemInteract) :!class /Script/SoundLibrary.SoundLibrarySubsystem))))

(fn ConstructionScript
  (|CallParentFunction)
  (Utilities|Array|AddUnique (Variables|Actor|GetTags) "Lackluster.Material.TieredAmmoAthena"))
